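; Thrown gift-box projectile that bounces off pawns and props, then deploys a grid-snapped present building containing loot, facing the thrower's cardinal direction.
; On explosion, a box trace selects a normal present, or a taller variant over terrain or water, then clears cells and spawns the builder.
; It also reports gifting quests from overlapped players, drops loot-table pickups with ammo, and destroys obstructing props gradually on a timer.
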(event EventBeginPlay
  ;[1 1 1] Lookup build delay row. Determine which cardinal direction the tower should face.
  (Audio|PlaySoundatLocation (Variables|Cosmetic|GetThrowSound) (Transformation|GetWorldLocation (Variables|Collision|GetCapsuleComponent)))
  (CallFunction|CalculateCardinalDirection)
  (CallFunction|OnBeginPlayQuestCheck))

(event EventEndPlay
  (params (EndPlayReason :type EEndPlayReason%20Enum :default "Destroyed"))
  (bind _buildingstructuralsupportsystem (Building|GetBuildingStructuralSupportSystem))
  ;[1 1 1] Destroy any remaining actors in our destroy list if needed
  (switch Utilities|FlowControl|Switch|SwitchonEEndPlayReason EndPlayReason
    (:Destroyed
      (Building|StartActorRemovalBatch _buildingstructuralsupportsystem)
      (for _array_element (Variables|Tower|GetBuildingActorsToDestroy)
        (Actor|DestroyActor _array_element))
      (Building|StopActorRemovalBatch _buildingstructuralsupportsystem))))

(event Collision|EventHit
  (params (MyComp :type Primitive%20Component%20Object%20Reference) (Other :type Actor%20Object%20Reference) (OtherComp :type Primitive%20Component%20Object%20Reference) (bSelfMoved :type Boolean) (HitLocation :type Vector :default "0, 0, 0") (HitNormal :type Vector :default "0, 0, 0") (NormalImpulse :type Vector :default "0, 0, 0") (Hit :type Hit%20Result%20Structure%20%28by%20ref%29))
  (bind (_asbuilding_prop _bsuccess) (Utilities|Casting|CastToBuildingProp Other))
  (bind (_asfort_athena_supply_drop _bsuccess_2) (Utilities|Casting|CastToFortAthenaSupplyDrop Other))
  (bind (_asfort_projectile_base _bsuccess_3) (Utilities|Casting|CastToFortProjectileBase Other))
  (bind (_asathena_supply_drop_balloon _bsuccess_4) (Utilities|Casting|CastToAthenaSupplyDropBalloon Other))
  (bind (_asfort_pawn _bsuccess_5) (Utilities|Casting|CastToFortPawn Other))
  ;[1 1 1] Bounce off of pawns and walls
  (if (Math|Boolean|ORBoolean _bsuccess _bsuccess_5 _bsuccess_2 _bsuccess_4 _bsuccess_3)
    (CallFunction|HandleBounce HitLocation)
    (Utilities|FlowControl|DoOnce false
      (:Completed
        (Utilities|FlowControl|Delay (Variables|Cosmetic|GetBounceTimeout)))))
  (Utilities|FlowControl|DoOnce false
    (:Completed
      (Game|Components|ProjectileMovement|StopSimulating (Variables|Movement|GetProjectileMovementComponent) Hit))))

(event Custom|Handle%20Bounce
  (params (Hit%20Location :type Vector))
  (bind _bouncesound (Variables|Cosmetic|GetBounceSound))
  (bind _lastbouncelocation (Variables|Cosmetic|GetLastBounceLocation))
  ;[1 1 1] Play audio feedback, etc when bouncing
  (if (and (> (Math|Vector|VectorLengthSquared (- Hit%20Location (Variables|Cosmetic|GetLastBounceLocation))) (Math|Float|Square (Variables|Cosmetic|GetMinBounceDistance))) (> (- (Utilities|Time|GetTimeSeconds) (Variables|Cosmetic|GetLastBounceTime)) (Variables|Cosmetic|GetMinTimeBetweenBounces)))
    (Variables|Cosmetic|SetLastBounceLocation Hit%20Location)
    (Variables|Cosmetic|SetLastBounceTime (Utilities|Time|GetTimeSeconds))
    (Audio|PlaySoundatLocation _bouncesound _lastbouncelocation)
    (Fortnite|ShowSoundIndicatorLocalforTeam :Location _lastbouncelocation :MaxAudibleDistance (Variables|/Script/Engine.SoundBase|GetMaxDistance _bouncesound) :IndicatorType "Gunshot" :Team (Team|GetActorTeam self) :Affiliations (Utilities|Array|MakeArray "Hostile" "Neutral") :Instigator self :Tint "(R=0.000000,G=0.000000,B=0.000000,A=0.000000)" :AttenuationMaxRangeMultiplier 1.0 :AttenuationIntensityMultiplier 1.0)))

(event EventOnStop
  (params (Hit :type Hit%20Result%20Structure%20%28by%20ref%29))
  ;[1 1 1] Initializing variables, stop rotation, and snap projectile to grid before "exploding"
  (Components|Activation|Deactivate (Variables|Components|GetTrail))
  (Rendering|SetVisibility (Variables|Components|GetMesh) false false)
  (CallFunction|SetBaseDestination Hit)
  (Components|Movement|SetUpdatedComponent (Variables|Components|GetRotatingMovement) 0)
  (Transformation|SetActorTransform self (Math|Transform|MakeTransform (Fortnite|SnaptoFortBuildingGrid (Transformation|GetActorLocation) (Math|Vector|MakeVector 0.0 0.0 (Variables|SnaptoGrid|GetGridZOffset)) "0.000000,0.000000,192.000000") (Math|Rotator|MakeRotator 0.0 0.0 (Utilities|Map|Find (Variables|Orientation|GetOrientationYawOffsets) (Variables|Orientation|GetCardinalDirection)))) false false)
  (Delivery|DoExplosion))

(event EventOnExploded
  (params (HitActors :type Array%20of%20Actor%20Object%20References) (HitResults :type Array%20of%20Hit%20Result%20Structures))
  (bind _returnvalue (Transformation|GetActorLocation))
  (bind _returnvalue_2 (- (Variables|BuildingAndClearing|GetBaseDestination) "0, 0,192.000000"))
  (bind _cardinaldirection (Variables|Orientation|GetCardinalDirection))
  (bind _orientationyawoffsets (Variables|Orientation|GetOrientationYawOffsets))
  (bind _value (Utilities|Map|Find _orientationyawoffsets _cardinaldirection))
  (bind _gridexplosionoffset (Variables|Cosmetic|GetGridExplosionOffset))
  (bind _returnvalue_3 (+ _returnvalue _gridexplosionoffset))
  (bind (_outlocation _outrotation) (Utilities|RotateSpawnLocation (Variables|BuildingAndClearing|GetBaseDestination) (Variables|TracedActors|GetLootSpawnLocation) (Variables|TracedActors|GetLootSpawnRotation)))
  (bind _basedestination (Variables|BuildingAndClearing|GetBaseDestination))
  (bind _cue_distantsound (Variables|Cosmetic|GetCue_DistantSound))
  ;[1 1 1] Trace to see if there is enough room to spawn loot at the location.
  (CallFunction|HandleProps HitActors)
  (bind _outhits (Collision|MultiBoxTraceByChannel (+ (Variables|TerrainTrace|GetDoorTraceOffsetHigh) _outlocation) (+ (Variables|TerrainTrace|GetDoorTraceOffsetLow) _outlocation) "32.000000,32.000000,32.000000" _outrotation "TraceTypeQuery10" false (Variables|Tower|GetBuildingActorsToIgnore)))
  ;[1 1 1] Check if any hits are against terrain instead of building actors
  (for _array_element _outhits
    (Development|PrintString (Utilities|String|Append "Present Hit: " (Utilities|GetDisplayName (.HitActor (Collision|BreakHitResult _array_element)))) (Variables|Default|GetIsDebug) (Variables|Default|GetIsDebug) "(R=0.000000,G=0.660000,B=1.000000,A=1.000000)" 1e+01)
    (Utilities|Casting|CastToBuildingActor (.HitActor (Collision|BreakHitResult _array_element))
      (:then
        (Variables|Default|SetB%20Hit%20Water (Utilities|Enum|Equal(Enum) (Variables|/Script/PhysicsCore.PhysicalMaterial|GetSurfaceType (.PhysMat (Collision|BreakHitResult _array_element))) "SurfaceType8")))
      (:CastFailed
        (Variables|TerrainTrace|Setb_HitTerrain true))))
  (Networking|CanExecuteCosmeticEvents
    (:True
      ;[1 1 1] OnExploded - Emitter, Sound
      (Niagara|SpawnSystematLocation (select (Variables|TerrainTrace|Getb_HitTerrain) (Variables|Cosmetic|GetNS_DeployTall) (Variables|Cosmetic|GetNS_Deploy)) (- _basedestination (select (Variables|TerrainTrace|Getb_HitTerrain) (Variables|Cosmetic|GetOffset_ParticleSystem_Tall) (Variables|Cosmetic|GetOffset_ParticleSystem))) (Math|Rotator|MakeRotator 0.0 0.0 (Utilities|Map|Find (Variables|Orientation|GetOrientationYawOffsets) (Variables|Orientation|GetCardinalDirection))))
      (Camera|PlayWorldCameraShake "/Game/Athena/Items/Gameplay/BackPacks/CarminePack/FX/B_Ground_CameraShake_Heavy.B_Ground_CameraShake_Heavy_C" _returnvalue_2 8e+02 4.5e+03 1.8)
      (Fortnite|PlayLocalForceFeedbackatLocationMulti "/Game/Weapons/ForceFeedback/Explosions/FF_Generic_Explosion_Near.FF_Generic_Explosion_Near" "/Game/Weapons/ForceFeedback/Explosions/FF_Generic_Explosion_Far.FF_Generic_Explosion_Far" _returnvalue_2 1e+03 2.5e+03)
      (Audio|Components|Audio|FadeOut (Variables|Components|GetInAirLoopComponent) 0.2 0.0 "Linear" :!class /Script/Engine.AudioComponent)
      (if (Pawn|IsLocallyControlled (Game|GetInstigator))
        (Audio|PlaySoundatLocation (Variables|Cosmetic|GetCue_CloseSound) _returnvalue_3) ; Play Close Explosion
        (else
          (Audio|PlaySoundatLocation _cue_distantsound _returnvalue_3) ; Play Distant Explosion
        ))
      (Fortnite|ShowSoundIndicatorLocalforTeam :Location _returnvalue_3 :MaxAudibleDistance (Variables|/Script/Engine.SoundBase|GetMaxDistance _cue_distantsound) :IndicatorType "Gunshot" :Team (Team|GetActorTeam self) :Affiliations (Utilities|Array|MakeArray "Hostile" "Neutral") :Instigator self :Tint "(R=0.000000,G=0.000000,B=0.000000,A=0.000000)" :AttenuationMaxRangeMultiplier 1.0 :AttenuationIntensityMultiplier 1.0))
    (:False
      (Utilities|FlowControl|Delay (Variables|Default|GetSpawnDelay))
      (if (Variables|TerrainTrace|Getb_HitTerrain)
        _
        (elif (Variables|Default|GetB%20Hit%20Water)
          _
          (else
            (Tower|ClearGridCells (Variables|BuildingandClearing|GetClearedCells_Normal) true (Variables|BuildingAndClearing|GetBaseDestination) _value)
            (CallFunction|CallDynamicBuilder (Variables|BuildingandClearing|GetBuilderClass) (Variables|BuildingAndClearing|GetBaseDestination))
            ;[0 0.5666 1] Hook in for quests
            (CallFunction|CheckforGiftedTargets (+ (Variables|BuildingAndClearing|GetBaseDestination) (Variables|Quests|GetQuestTraceOffset)) (Variables|Quests|GetQuestTraceExtent))
            (if (Variables|TracedActors|Getb_SpawnTracedActors) :!label _label
              ;[1 1 1] Function that spawns the loot
              (Utilities|FlowControl|Delay (Variables|TracedActors|GetTracedActorBuildDelay))
              (CallFunction|SpawnContents (Variables|BuildingAndClearing|GetBaseDestination)))
            ;[1 1 1] loop a timer deleting an actor every frame or so
            (bind _returnvalue_4 (Utilities|Time|SetTimerbyEvent (delegate Custom|Delete%20another%20actor) 0.03 true)))))
      ;[1 1 1] Loot would spawn underground in a 1x1x1present, so we spawn a taller 1x1x2 present.
      (Tower|ClearGridCells (Variables|BuildingandClearing|GetClearedCells_Underground) true (Variables|BuildingAndClearing|GetBaseDestination) _value)
      (CallFunction|CallDynamicBuilder (Variables|BuildingandClearing|GetBuilderClassBasement) (Variables|BuildingAndClearing|GetBaseDestination))
      (Variables|BuildingAndClearing|SetBaseDestination (+ (Variables|BuildingAndClearing|GetBaseDestination) (Variables|BuildingandClearing|GetBaseDestinationOffsetUnderground)))
      (CallFunction|CheckforGiftedTargets (+ (Variables|BuildingAndClearing|GetBaseDestination) (Variables|Quests|GetQuestTractOffsetTall)) (Variables|Quests|GetQuestTraceExtentTall))
      (goto _label))))

(event Custom|Delete%20another%20actor
  (bind _buildingactorstodestroy (Variables|Tower|GetBuildingActorsToDestroy))
  ;[1 1 1] loop a timer deleting an actor every frame or so
  (if (> (Utilities|Array|Length _buildingactorstodestroy) 0)
    (bind (_array_element _array_index) (Utilities|Array|ForEachLoopwithBreak _buildingactorstodestroy :!label _label)
      (:LoopBody
        (Utilities|IsValid _array_element
          (:Is%20Valid
            (if (Variables|Actor|GetbCanBeDamaged _array_element)
              (Actor|DestroyActor _array_element)))
          (:Is%20Not%20Valid
            _))
        (Utilities|Array|RemoveIndex _buildingactorstodestroy _array_index)
        (goto _label Break)))
    (else
      (Utilities|Time|ClearandInvalidateTimerbyHandle (stale EventOnExploded _returnvalue_4)))))

(event Custom|OnBeginPlayQuestCheck)

(event Custom|Play%20Quest%20Success%20FX
  (bind _controllerwhothrew (Variables|Quests|GetControllerWhoThrew))
  ;[0.5718 1 0.6507] Check Quests Triggered off of BeginPlay
  ;;[1 1 1] Play Quest Success VFX
  (Networking|SwitchHasAuthority
    (:Remote
      (Utilities|IsValid _controllerwhothrew
        (:Is%20Valid
          (bind _asfort_player_controller_athena (Utilities|Casting|CastToFortPlayerControllerAthena _controllerwhothrew))
          (Audio|PlaySoundatLocation (Variables|Quests|GetQuestSuccessSound) (Transformation|GetActorLocation (Game|Player|GetPlayerPawn _asfort_player_controller_athena)))
          (Effects|SpawnFXSystemAttached (Variables|Quests|GetQuestSuccessEmitter) (Variables|/Script/Engine.Character|GetMesh (Game|Player|GetPlayerPawn _asfort_player_controller_athena))))))))

(fn CalculateCardinalDirection
  (bind _returnvalue (Game|GetInstigator))
  (bind _returnvalue_2 (Transformation|GetActorForwardVector _returnvalue))
  (bind _x (.x _returnvalue_2))
  (bind _y (.y _returnvalue_2))
  (if (>= (Math|Float|Absolute(Float) _x) (Math|Float|Absolute(Float) _y))
    (if (>= _x 0.0)
      (Variables|Orientation|SetCardinalDirection "NewEnumerator0")
      (else
        (Variables|Orientation|SetCardinalDirection "NewEnumerator1")))
    (elif (>= _y 0.0)
      (Variables|Orientation|SetCardinalDirection "NewEnumerator2")
      (else
        (Variables|Orientation|SetCardinalDirection "NewEnumerator3")))))

(fn CallDynamicBuilder
  (params (Class :type Building%20Actor%20Class%20Reference) (Location :type Vector))
  (bind _returnvalue (Building|SpawnBuildingActor Class (Math|Transform|MakeTransform Location (Math|Rotator|MakeRotator 0.0 0.0 (Utilities|Map|Find (Variables|Orientation|GetOrientationYawOffsets) (Variables|Orientation|GetCardinalDirection))))))
  (bind _asbuilding_actor (Utilities|Casting|CastToBuildingActor _returnvalue))
  (bind _asfort_player_controller (Utilities|Casting|CastToFortPlayerController (Game|GetInstigatorController)))
  (Building|InitializeKismetSpawnedBuildingActor _asbuilding_actor 0 _asfort_player_controller true 0 false))

(fn CheckForGiftedTargets
  (params (TraceLocation :type Vector) (BoxExtent :type Vector))
  (bind _returnvalue (Game|GetInstigator))
  (bind _instigatingpawn (LocalVariables|GetInstigatingPawn))
  (Utilities|IsValid _returnvalue
    (:Is%20Valid
      (bind _asfort_player_pawn (Utilities|Casting|CastToFortPlayerPawn _returnvalue))
      (LocalVariables|SetInstigatingPawn _asfort_player_pawn))
    (:Is%20Not%20Valid
      _))
  (bind _asfort_player_controller (Utilities|Casting|CastToFortPlayerController (Game|GetInstigatorController)))
  (LocalVariables|SetInstigatorController _asfort_player_controller)
  (LocalVariables|SetCurrentQuestCase (LocalVariables|GetQuestCase_PresentThrown))
  (if (Variables|Default|GetIsDebug)
    (Debug|DebugBoxSendtoAll TraceLocation BoxExtent (LocalVariables|GetLineColor) true 1e+01 0 32.0))
  (bind _outactors (Collision|BoxOverlapActors TraceLocation BoxExtent (Utilities|Array|MakeArray "ObjectTypeQuery3") "/Script/FortniteGame.FortPlayerPawn" :!class /Script/Engine.KismetSystemLibrary))
  (for _array_element _outactors
    (bind _asfort_player_pawn_2 (Utilities|Casting|CastToFortPlayerPawn _array_element))
    (if (Team|OnSameTeam (LocalVariables|GetInstigatorController) _asfort_player_pawn_2)
      (Utilities|IsValid _instigatingpawn
        (:Is%20Valid
          (if (Utilities|Equal(Object) _instigatingpawn _asfort_player_pawn_2)
            (LocalVariables|SetCurrentQuestCase (LocalVariables|GetQuestCase_GiftedSelf))
            (else
              (LocalVariables|SetCurrentQuestCase (LocalVariables|GetQuestCase_GiftedAlly) :!label _label)))
          (Development|PrintString (Utilities|Select :Index (LocalVariables|GetCurrentQuestCase) :Option%200 "PresentThrown" :Option%201 "GiftedAlly" :Option%202 "GiftedEnemy" :Option%203 "GiftedSelf") (Variables|Default|GetIsDebug) (Variables|Default|GetIsDebug) :!label _label_2)
          ;[0 0.4797 1] Populate these arrays with quest info!
          ; (0 = present thrown, 1 = gifted ally, 2 = gifted enemy, 3 = gifted self)
          (Quest|GiveQuestUpdatetoPlayers (Utilities|Array|Get(acopy) (LocalVariables|GetFortQuestItemDefinitions) (LocalVariables|GetCurrentQuestCase)) (Utilities|Array|Get(acopy) (LocalVariables|GetQuestObjectiveBackendNames) (LocalVariables|GetCurrentQuestCase)) (Utilities|Array|Get(acopy) (LocalVariables|GetObjectiveStatEvents) (LocalVariables|GetCurrentQuestCase)) (Utilities|Array|MakeArray (LocalVariables|GetInstigatorController))))
        (:Is%20Not%20Valid
          _))
      (goto _label)
      (else
        (LocalVariables|SetCurrentQuestCase (LocalVariables|GetQuestCase_GiftedEnemy))))
    (goto _label_2))
  (LocalVariables|SetCurrentQuestCase (LocalVariables|GetQuestCase_PresentThrown))
  (goto _label_2))

(fn ClearAndBuild
  (params (ClearedGridCells :type Array%20of%20Vectors) (BGAClassArray :type Array%20of%20Actor%20Class%20References) (BGALocationArray :type Array%20of%20Vectors) (BGARotationArray :type Array%20of%20Rotators) (DestroyFloorsQuickly :type Boolean))
  (for _array_element ClearedGridCells
    (Tower|DestroyActorsinGridCellLocation (Building|GetBuildingStructuralSupportSystem) (+ (Variables|BuildingAndClearing|GetBaseDestination) (Math|Vector|RotateVectorAroundAxis _array_element (Utilities|Map|Find (Variables|Orientation|GetOrientationYawOffsets) (Variables|Orientation|GetCardinalDirection)) "0, 0,1.000000")) DestroyFloorsQuickly))
  (CallFunction|CreateBaseSection (Variables|BuildingAndClearing|GetBaseDestination) BGAClassArray BGALocationArray BGARotationArray))

(fn ClearOnly
  (params (ClearedGridCells :type Array%20of%20Vectors) (DestroyFloorsQuickly :type Boolean))
  (bind _buildingstructuralsupportsystem (Building|GetBuildingStructuralSupportSystem))
  (Building|StartActorRemovalBatch _buildingstructuralsupportsystem)
  (for _array_element ClearedGridCells
    (Tower|DestroyActorsinGridCellLocation (Building|GetBuildingStructuralSupportSystem) (+ (Variables|BuildingAndClearing|GetBaseDestination) (Math|Vector|RotateVectorAroundAxis _array_element (Utilities|Map|Find (Variables|Orientation|GetOrientationYawOffsets) (Variables|Orientation|GetCardinalDirection)) "0, 0,1.000000")) DestroyFloorsQuickly))
  (Building|StopActorRemovalBatch _buildingstructuralsupportsystem))

(fn CreateBaseSection
  (params (ReferenceLocation :type Vector) (BGAClassArray :type Array%20of%20Actor%20Class%20References) (BGALocationArray :type Array%20of%20Vectors) (BGARotationArray :type Array%20of%20Rotators))
  (for _index (range (Utilities|Array|Length BGAClassArray))
    (bind _returnvalue (Game|SpawnActorfromClass (Utilities|Array|Get(acopy) BGAClassArray _index) (Math|Transform|MakeTransform (Math|Vector|vector+vector ReferenceLocation (Math|Vector|RotateVectorAroundAxis (Utilities|Array|Get(aref) BGALocationArray _index) (Utilities|Map|Find (Variables|Orientation|GetOrientationYawOffsets) (Variables|Orientation|GetCardinalDirection)) "0, 0,1.000000") (Utilities|Map|Find (Variables|Orientation|GetOrientationLocationOffsets) (Variables|Orientation|GetCardinalDirection))) (Math|Rotator|MakeRotator 0.0 0.0 (+ (.yaw (Utilities|Array|Get(aref) BGARotationArray _index)) (Utilities|Map|Find (Variables|Orientation|GetOrientationYawOffsets) (Variables|Orientation|GetCardinalDirection))))) "Undefined" "SelectDefaultAtRuntime" 0 0))
    (bind _asbuilding_static_mesh (Utilities|Casting|CastToBuildingSMActor _returnvalue))
    (Class|BuildingActor|SetTeam _asbuilding_static_mesh (Team|GetActorTeam (Actor|GetOwner)))
    (Building|InitializeKismetSpawnedBuildingActor _asbuilding_static_mesh 0 0 true 0 false)
    (bind _asbuilding_trap (Utilities|Casting|CastToBuildingTrap _asbuilding_static_mesh)
      (:then
        (Trap|SpawnedfromItemDefinition _asbuilding_trap "/Game/Athena/Items/Traps/DamageTrap/TID_Floor_Spikes_Athena_R_T01.TID_Floor_Spikes_Athena_R_T01" 1))
      (:CastFailed
        _))
    (if (not (and (Building|IsWorldLocValid (Building|GetBuildingStructuralSupportSystem) (+ (Transformation|GetActorLocation _asbuilding_static_mesh) "0, 0,384.000000")) (Building|IsWorldLocValid (Building|GetBuildingStructuralSupportSystem) (+ (Transformation|GetActorLocation _asbuilding_static_mesh) "0, 0,-384.000000"))))
      (Actor|DestroyActor _asbuilding_static_mesh))))

(fn HandleProps
  (params (Array :type Array%20of%20Actor%20Object%20References))
  (for _array_element Array
    (Utilities|Casting|CastToBuildingProp _array_element
      (:then
        (Utilities|Casting|CastToTiered_Chest_Athena _array_element
          (:then
            _)
          (:CastFailed
            (Utilities|Casting|CastToBuildingPropBigHealth _array_element
              (:then
                _)
              (:CastFailed
                (if (Variables|Actor|GetbCanBeDamaged _array_element) :!label _label_2
                  (Utilities|Array|Add (Variables|Tower|GetBuildingActorsToDestroy) _array_element))))))
        (Utilities|Array|Add (Variables|Tower|GetBuildingActorsToIgnore) _array_element :!label _label))
      (:CastFailed
        (Utilities|Casting|CastToBP_DamageBalloon_Athena _array_element
          (:then
            _)
          (:CastFailed
            (Utilities|Casting|CastToBuildingActor _array_element)
            (goto _label)))))
    (goto _label_2)))

(fn SetBaseDestination
  (params (Hit :type Hit%20Result%20Structure%20%28by%20ref%29))
  (Variables|BuildingAndClearing|SetBaseDestination (Fortnite|SnaptoFortBuildingGrid (Transformation|GetActorLocation) (Math|Vector|MakeVector 0.0 0.0 (Variables|SnaptoGrid|GetGridZOffset)) (Math|Vector|MakeVector 0.0 (Variables|SnaptoGrid|GetGridYOffset)))))

(fn SpawnContents
  (params (ReferenceLocation :type Vector))
  (bind _lootspawnrotation (Variables|TracedActors|GetLootSpawnRotation))
  (bind _lootspawnlocation (Variables|TracedActors|GetLootSpawnLocation))
  (bind _outlocation (Utilities|RotateSpawnLocation ReferenceLocation _lootspawnlocation _lootspawnrotation))
  (bind (_outhit _returnvalue) (Collision|SphereTraceByChannel (+ _outlocation "0, 0,320.000000") (+ _outlocation "0, 0,-384.000000") 32.0 "TraceTypeQuery10" false (Variables|Tower|GetBuildingActorsToIgnore)))
  (Development|PrintString (Utilities|String|Append "Tower Grenade Tire trace hit: " (Utilities|GetDisplayName (.HitActor (Collision|BreakHitResult _outhit)))) (Variables|Default|GetIsDebug) (Variables|Default|GetIsDebug))
  (bind _outloottodrop (Loot|PickLootDrops (Variables|Default|GetLoot%20Table) 0 -1 "(GameplayTags=)"))
  (if true
    (for _array_element _outloottodrop
      (bind _asfort_world_item_definition (Utilities|Casting|CastToFortWorldItemDefinition (Class|FortItemFunctionLibrary|BreakFortItemEntry _array_element)))
      (Inventory|SpawnPickupInWorld _asfort_world_item_definition (.Count (Class|FortItemFunctionLibrary|BreakFortItemEntry _array_element)) (select _returnvalue (.Location (Collision|BreakHitResult _outhit)) _outlocation) "0.000000,0.000000,1.000000" 0 true false) ; Pistol C
      (bind _asfort_weapon_item_definition (Utilities|Casting|CastToFortWeaponItemDefinition _asfort_world_item_definition))
      (Fortnite|Weapon|CreateTossAmmoPickupforWeaponItemDefinitionatLocation _asfort_weapon_item_definition (Variables|Default|GetT_SourceTagsAmmo) (select _returnvalue (.Location (Collision|BreakHitResult _outhit)) _outlocation)))))

(fn SpawnProps
  (params (ReferenceLocation :type Vector))
  (for _index (range (Utilities|Array|Length (Variables|TracedActors|GetBGAClasses_Traced)))
    (bind _returnvalue (Game|SpawnActorfromClass (Utilities|Array|Get(acopy) (Variables|TracedActors|GetBGAClasses_Traced) _index) (Math|Transform|MakeTransform (Utilities|RotateSpawnLocation ReferenceLocation (Utilities|Array|Get(aref) (Variables|TracedActors|GetBGALocations_Traced) _index) (Utilities|Array|Get(aref) (Variables|TracedActors|GetBGARotations_Traced) _index)) (.OutRotation (Utilities|RotateSpawnLocation ReferenceLocation (Utilities|Array|Get(aref) (Variables|TracedActors|GetBGALocations_Traced) _index) (Utilities|Array|Get(aref) (Variables|TracedActors|GetBGARotations_Traced) _index))) (Utilities|Array|Get(aref) (Variables|TracedActors|GetBGAScales_Traced) _index)) "Undefined" "SelectDefaultAtRuntime" 0 0))
    (bind _asbuilding_actor (Utilities|Casting|CastToBuildingActor _returnvalue))
    (Building|InitializeKismetSpawnedBuildingActor _asbuilding_actor 0 0 true 0 false)
    (if (not (and (Building|IsWorldLocValid (Building|GetBuildingStructuralSupportSystem) (+ (Transformation|GetActorLocation _returnvalue) "0, 0,384.000000")) (Building|IsWorldLocValid (Building|GetBuildingStructuralSupportSystem) (+ (Transformation|GetActorLocation _returnvalue) "0, 0,-384.000000"))))
      (Actor|DestroyActor _returnvalue))))

(fn ConstructionScript
  (Networking|CanExecuteCosmeticEvents
    (:True
      (Components|Movement|SetUpdatedComponent (Variables|Components|GetRotatingMovement) (Variables|Components|GetMesh)))))
