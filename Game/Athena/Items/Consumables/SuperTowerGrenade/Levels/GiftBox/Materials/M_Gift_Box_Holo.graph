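# Unlit, additive, two-sided hologram material for the gift box variant of the SuperTowerGrenade consumable.
# It sums a blueprint metal texture and Fresnel rim, both tinted by the ExternalColor parameter, with blue scanlines panning in screen space.
# The sum feeds the MF_FX_EmissiveAdjustments function.
# A world-position banding chain (Frac, Floor, Clamp) is present but feeds nothing.
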Material /Game/Athena/Items/Consumables/SuperTowerGrenade/Levels/GiftBox/Materials/M_Gift_Box_Holo
graph {
 properties {
  bEnableResponsiveAA=true
  bUsedWithClothing=true
  bUsedWithMeshDeformer=true
  bUsedWithMeshParticles=true
  bUsedWithMorphTargets=true
  bUsedWithSkeletalMesh=true
  blendMode=BLEND_Additive
  pixelDepthOffsetMode=PDOM_Legacy
  previewMesh=/Game/Weapons/Grenades/Tower_Grenade/Meshes/SM_Tower_Preview
  shadingModel=MSM_Unlit
  twoSided=true
  usedShadingModels=Unlit
 }
 nodes [
  n1 = ComponentMask(None=WorldPosition().XYZ) {b=true}
  n2 = Add(A=Frac(None=Multiply(A=@n1) {constB=0.4})) {constB=0.2}
  Clamp(None=Add(A=Floor(None=@n2)) {constB=0.4})
  n3 = TextureSample() {texture=/Game/Packages/Fortress_SharedMaterials/BlueprintTextureMetal}
  n4 = VectorParameter() {parameterName=ExternalColor, defaultValue=<rgba 0.1213 0.614 1.5 1>}
  n5 = Multiply(A=Fresnel() {exponent=3}, B=@n4.RGB)
  n6 = Multiply(A=ScreenPosition().ViewportUV) {constB=4}
  n7 = Time()
  n8 = Multiply(
   A=Constant3Vector() {constant=<rgba 0.0937 0.1719 0.25 1>}
   B=TextureSample(UVs=Panner(Coordinate=@n6, Time=@n7) {speedX=0.1, speedY=0.25}) {texture=/Game/Effects/Fort_Effects/Textures/Tile/T_ScanLine_Analog}.R
  )
  MaterialFunctionCall(Emissive=Add(A=Multiply(A=@n3.R, B=@n4.RGB), B=Add(A=@n5, B=@n8))) {materialFunction=&MF_FX_EmissiveAdjustments}
 ]
}
