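# Parent material for the galaxy-styled birthday present in the SuperTowerGrenade gift box level, using Default Lit with pixel animation.
# Texture parameters build base material attributes, with SpecularMasks green driving Metallic, Specular and Roughness despite its note describing separate RGB channels.
# MF_Galaxy_Birthday layers the galaxy effect onto them, taking its Alpha from a selectable SkinFX_Map channel.
# MF_FN_DestructibleDamage processes that result last.
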Material /Game/Athena/Items/Consumables/SuperTowerGrenade/Levels/GiftBox/Materials/M_Present_Birthday_Galaxy_Parent
graph {
 properties {bAutomaticallySetUsageInEditor=false, bHasPixelAnimation=true, bUseMaterialAttributes=true, usedShadingModels=Default Lit}
 nodes [
  n1 = TextureSampleParameter2D() {desc="R = Specular\r\nG = Metal\r\nB = Roughness", parameterName=SpecularMasks, texture=&T_BlacksRGB}
  n2 = SetMaterialAttributes!(
   MaterialAttributes
   Base Color=TextureSampleParameter2D() {parameterName=Diffuse, texture=&T_BlacksRGB}.RGB
   Emissive Color=TextureSampleParameter2D() {parameterName=Emissive, texture=&T_BlacksRGB}.RGB
   Metallic=@n1.G
   Specular=@n1.G
   Roughness=@n1.G
   Normal=TextureSampleParameter2D() {parameterName=Normal Map, samplerType=SAMPLERTYPE_Normal, texture=&FlatNormalBlue}.RGB
  ) {attributeSetTypes [
    #1
    #2
    #3
    #4
    #5
    #6
   ]}
  n3 = TextureSampleParameter2D() {parameterName=SkinFX_Map, texture=&T_BlacksRGB}
  MaterialFunctionCall(Material_In=MaterialFunctionCall!(
    In=@n2
    Warp Mask
    GalaxyAlternateFresnel
    Alpha=StaticComponentMaskParameter(None=@n3.RGBA) {defaultR=true, parameterName=SkinFX_Map_Channel}
   ) {materialFunction=&/MF_Galaxy_Birthday}.Result) {materialFunction=&MF_FN_DestructibleDamage}
 ]
}
