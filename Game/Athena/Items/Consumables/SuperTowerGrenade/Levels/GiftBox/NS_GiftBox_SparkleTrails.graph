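# Niagara trail effect for a moving gift box in the SuperTowerGrenade item, combining confetti, sparkles, a ribbon and a glow halo.
# Two confetti emitters and a sparkle emitter spawn sprites by distance traveled, after a short probability ramp, throwing them opposite owner velocity into curl noise.
# A distance-spawned ribbon gets a strong upward force, scaled by execution index against heavy drag.
# A burst-spawned mesh halo scales up, holds and shrinks.
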NiagaraSystem /Game/Athena/Items/Consumables/SuperTowerGrenade/Levels/GiftBox/NS_GiftBox_SparkleTrails
dependencies {
 usedDataInterfaces [
  &NiagaraDataInterfaceMeshRendererInfo
  &NiagaraDataInterfaceCurve
  &NiagaraDataInterfaceColorCurve
  &NiagaraDataInterfaceCurlNoise
  &NiagaraDataInterfaceVector2DCurve
 ]
 usedDynamicInputs []
 usedModules [&SystemState]
}
emitters {
 Confetti {
  emitterUpdateScript [
   &EmitterState(Life Cycle Mode=::NewEnumerator1, Loop Behavior=::NewEnumerator1, Loop Duration=3, UseLoopDelay=false)
   &SpawnPerUnit(
    Module.Use Max Movement Threshold=false
    Module.Use Movement Tolerance=false
    Spawn Probability:NiagaraFloat=&FloatFromCurve(FloatCurve=DataInterface() {propertyValues {Curve {keys [<at 0.075 0>,<at 0.15 1>]}}}, CurveIndex=$Emitter.NormalizedLoopAge)
    Use Spawn Probability=true
   )
  ]
  particleSpawnScript [
   &InitializeParticle(
    Color Mode=::NewEnumerator0
    Lifetime Mode=::NewEnumerator1
    Mass Mode=::NewEnumerator3
    Sprite Rotation Mode=::NewEnumerator2
    Sprite Size Mode=::NewEnumerator1
    Uniform Sprite Size Max=15
    Uniform Sprite Size Min=6
    UsePositionOffset=false
    Write Material Random=true
   )
   &SphereLocation(
    MaskSpawnGroup=false
    Override Seed=false
    Sphere Radius=5
    Surface Only=false
    UseNonUniformScale=false
    UseOffset=false
    UseSphereOrientation=false
   )
   &AddVelocityInCone(
    Cone Angle=20
    Cone Axis=&Multiply_VectorByFloat(Vector=$Engine.Owner.Velocity, Float=-1)
    Module.Use Velocity Falloff On Cone Axis=false
    Velocity Distribution Along Cone Axis=0.75
    Velocity Strength=&RandomRangeFloat(Minimum=200, Maximum=50)
   )
  ]
  particleUpdateScript [
   &ParticleState()
   &GravityForce(Gravity=<0 0 -35>)
   &Drag(Drag=&RandomRangeFloat(Minimum=0.1, Maximum=0.5), Use Linear Drag=true, Use Rotational Drag=false)
   &Color(Color=&ColorFromCurve(ColorCurve=DataInterface() {propertyValues {
       AlphaCurve {keys [{arriveTangent=-10.8596, interpMode=RCIM_Cubic, leaveTangent=-10.8596, time=0.9079, value=1}]}
       BlueCurve {keys [
         <at 0.2697 0 RCIM_Cubic>
         <at 0.4724 0 RCIM_Cubic>
         {arriveTangent=0.1899, interpMode=RCIM_Cubic, leaveTangent=0.1899, time=0.6895, value=0.036}
        ]}
       GreenCurve {keys [
         {arriveTangent=3.6319, interpMode=RCIM_Cubic, leaveTangent=3.6319, time=0.2697, value=0}
         {arriveTangent=3.6319, interpMode=RCIM_Cubic, leaveTangent=3.6319, time=0.4724, value=0.736}
         {arriveTangent=-0.5306, interpMode=RCIM_Cubic, leaveTangent=-0.5306, time=0.6895, value=0.6354}
        ]}
       RedCurve {keys [
         {arriveTangent=0.6169, interpMode=RCIM_Cubic, leaveTangent=0.6169, time=0.2697, value=0.875}
         {arriveTangent=0.6169, interpMode=RCIM_Cubic, leaveTangent=0.6169, time=0.4724, value=1}
         {arriveTangent=-5.2777, interpMode=RCIM_Cubic, leaveTangent=-5.2777, time=0.6895, value=0}
        ]}
      }}))
   /SetVariables(Particles.Color=&MultiplyLinearColor(
     Linear Color=$Particles.Color
     Scale Factor=&Vector4FromCurve(
      CurveIndex=&Divide_Float(A=$Particles.NormalizedAge, B=3)
      Vector4Curve=DataInterface() {propertyValues {
        AlphaCurve {keys [<at 0 0>,<at 0.0624 1>,<at 0.3556 1>,<at 1 0>]}
        BlueCurve {keys [<at 0.0208 15>,<at 0.5255 0.9801>,<at 1 0>]}
        GreenCurve {keys [<at 0.0208 5>,<at 0.5255 2>,<at 1 0>]}
        RedCurve {keys [<at 0.0208 0>,<at 0.5255 1>,<at 1 0>]}
       }}
     )
    )) {isSetParameters=true, module=SetVariables_#1}
   &CurlNoiseForce(Noise Frequency=4, Noise Strength=255, Pan Noise=true) {module=CurlNoiseForce001}
   &SolveForcesAndVelocity(Clamp Velocity=false, Limit Acceleration=false)
   &SpriteRotationRate(Rotation Rate=&RandomRangeFloat(Minimum=1, Maximum=3))
   &ScaleSpriteSize(Uniform Curve Sprite Scale=DataInterface() {propertyValues {Curve {keys [
        {arriveTangent=8.9686, interpMode=RCIM_Cubic, leaveTangent=8.9686, time=0, value=0}
        {arriveTangent=0.9375, interpMode=RCIM_Cubic, leaveTangent=0.9375, time=0.0842, value=0.7547}
        {arriveTangent=-1.2897, interpMode=RCIM_Cubic, leaveTangent=-1.2897, tangentMode=RCTM_User, time=0.8, value=0.75}
        {arriveTangent=-0.285, interpMode=RCIM_Cubic, leaveTangent=-0.285, tangentMode=RCTM_User, time=1, value=0}
       ]}}})
  ]
  rendererClasses [&NiagaraSpriteRendererProperties]
  renderers [{rendererClass=&NiagaraSpriteRendererProperties}]
 }
 Confetti001 {
  emitterUpdateScript [
   &EmitterState(Life Cycle Mode=::NewEnumerator1, Loop Behavior=::NewEnumerator1, Loop Duration=3, UseLoopDelay=false)
   &SpawnPerUnit(
    Module.Use Max Movement Threshold=false
    Module.Use Movement Tolerance=false
    Spawn Probability:NiagaraFloat=&FloatFromCurve(FloatCurve=DataInterface() {propertyValues {Curve {keys [<at 0.075 0>,<at 0.15 1>]}}}, CurveIndex=$Emitter.NormalizedLoopAge)
    Use Spawn Probability=true
   )
  ]
  particleSpawnScript [
   &InitializeParticle(
    Color Mode=::NewEnumerator0
    Lifetime Mode=::NewEnumerator1
    Mass Mode=::NewEnumerator3
    Sprite Rotation Mode=::NewEnumerator2
    Sprite Size Mode=::NewEnumerator1
    Uniform Sprite Size Max=15
    Uniform Sprite Size Min=6
    UsePositionOffset=false
    Write Material Random=true
   )
   &SphereLocation(
    MaskSpawnGroup=false
    Override Seed=false
    Sphere Radius=189
    Surface Only=false
    UseNonUniformScale=false
    UseOffset=false
    UseSphereOrientation=false
   )
   &AddVelocityInCone(
    Cone Angle=20
    Cone Axis=&Multiply_VectorByFloat(Vector=$Engine.Owner.Velocity, Float=-1)
    Module.Use Velocity Falloff On Cone Axis=false
    Velocity Distribution Along Cone Axis=0.75
    Velocity Strength=&RandomRangeFloat(Minimum=200, Maximum=50)
   )
  ]
  particleUpdateScript [
   &ParticleState()
   &GravityForce(Gravity=<0 0 -35>)
   &Drag(Drag=0.25, Use Linear Drag=true, Use Rotational Drag=false)
   &Color(Color=&ColorFromCurve(ColorCurve=DataInterface() {propertyValues {
       AlphaCurve {keys [{arriveTangent=-10.8596, interpMode=RCIM_Cubic, leaveTangent=-10.8596, time=0.9079, value=1}]}
       BlueCurve {keys [
         <at 0.2697 0 RCIM_Cubic>
         <at 0.4724 0 RCIM_Cubic>
         {arriveTangent=0.1899, interpMode=RCIM_Cubic, leaveTangent=0.1899, time=0.6895, value=0.036}
        ]}
       GreenCurve {keys [
         {arriveTangent=3.6319, interpMode=RCIM_Cubic, leaveTangent=3.6319, time=0.2697, value=0}
         {arriveTangent=3.6319, interpMode=RCIM_Cubic, leaveTangent=3.6319, time=0.4724, value=0.736}
         {arriveTangent=-0.5306, interpMode=RCIM_Cubic, leaveTangent=-0.5306, time=0.6895, value=0.6354}
        ]}
       RedCurve {keys [
         {arriveTangent=0.6169, interpMode=RCIM_Cubic, leaveTangent=0.6169, time=0.2697, value=0.875}
         {arriveTangent=0.6169, interpMode=RCIM_Cubic, leaveTangent=0.6169, time=0.4724, value=1}
         {arriveTangent=-5.2777, interpMode=RCIM_Cubic, leaveTangent=-5.2777, time=0.6895, value=0}
        ]}
      }}))
   /SetVariables(Particles.Color=&MultiplyLinearColor(
     Linear Color=$Particles.Color
     Scale Factor=&Vector4FromCurve(
      CurveIndex=&Divide_Float(A=$Particles.NormalizedAge, B=3)
      Vector4Curve=DataInterface() {propertyValues {
        AlphaCurve {keys [<at 0.3556 1>,<at 1 0>]}
        BlueCurve {keys [<at 0.3503 1>,<at 1 0>]}
        GreenCurve {keys [<at 0.3503 1>,<at 1 0>]}
        RedCurve {keys [<at 0.3503 1>,<at 1 0>]}
       }}
     )
    )) {isSetParameters=true}
   &CurlNoiseForce(Noise Frequency=4, Noise Strength=255, Pan Noise=true) {module=CurlNoiseForce001}
   &SolveForcesAndVelocity(Clamp Velocity=false, Limit Acceleration=false)
   &SpriteRotationRate(Rotation Rate=&RandomRangeFloat(Minimum=1, Maximum=3))
   &ScaleSpriteSize(Uniform Curve Sprite Scale=DataInterface() {propertyValues {Curve {keys [
        {arriveTangent=8.9686, interpMode=RCIM_Cubic, leaveTangent=8.9686, time=0, value=0}
        {arriveTangent=0.9375, interpMode=RCIM_Cubic, leaveTangent=0.9375, time=0.0842, value=0.7547}
        {arriveTangent=-1.2897, interpMode=RCIM_Cubic, leaveTangent=-1.2897, tangentMode=RCTM_User, time=0.8, value=0.75}
        {arriveTangent=-0.285, interpMode=RCIM_Cubic, leaveTangent=-0.285, tangentMode=RCTM_User, time=1, value=0}
       ]}}})
  ]
  rendererClasses [&NiagaraSpriteRendererProperties]
  renderers [{rendererClass=&NiagaraSpriteRendererProperties}]
 }
 Sparkes {
  emitterUpdateScript [
   &EmitterState(Life Cycle Mode=::NewEnumerator1, Loop Behavior=::NewEnumerator1, Loop Duration=3, UseLoopDelay=false)
   &SpawnPerUnit(
    Module.Use Max Movement Threshold=false
    Module.Use Movement Tolerance=false
    Spawn Probability:NiagaraFloat=&FloatFromCurve(FloatCurve=DataInterface() {propertyValues {Curve {keys [<at 0.04 0>,<at 0.15 1>]}}}, CurveIndex=$Emitter.NormalizedLoopAge)
    Spawn Spacing=10
    Use Spawn Probability=true
   )
  ]
  particleSpawnScript [
   &InitializeParticle(
    Color Mode=::NewEnumerator0
    Lifetime Max=3
    Lifetime Mode=::NewEnumerator1
    Mass Mode=::NewEnumerator3
    Sprite Rotation Mode=::NewEnumerator2
    Sprite Size Mode=::NewEnumerator1
    Uniform Sprite Size Max=22
    Uniform Sprite Size Min=6
    UsePositionOffset=false
    Write Material Random=true
   )
   &SphereLocation(
    MaskSpawnGroup=false
    Override Seed=false
    Sphere Radius=3
    Surface Only=false
    UseNonUniformScale=false
    UseOffset=false
    UseSphereOrientation=false
   )
   &AddVelocityInCone(
    Cone Axis=&Multiply_VectorByFloat(Vector=$Engine.Owner.Velocity, Float=-1)
    Module.Use Velocity Falloff On Cone Axis=false
    Velocity Distribution Along Cone Axis=0.75
    Velocity Strength=&RandomRangeFloat(Minimum=55, Maximum=222)
   )
  ]
  particleUpdateScript [
   &ParticleState()
   &GravityForce(Gravity=<0 0 -35>)
   &Drag(Drag=&RandomRangeFloat(Minimum=0.1, Maximum=0.3), Use Linear Drag=true, Use Rotational Drag=false)
   &Color(Color=&ColorFromCurve(ColorCurve=DataInterface() {propertyValues {
       AlphaCurve {keys [{arriveTangent=-10.8596, interpMode=RCIM_Cubic, leaveTangent=-10.8596, time=0.9079, value=1}]}
       BlueCurve {keys [
         <at 0.2697 0 RCIM_Cubic>
         <at 0.4724 0 RCIM_Cubic>
         {arriveTangent=0.1899, interpMode=RCIM_Cubic, leaveTangent=0.1899, time=0.6895, value=0.036}
        ]}
       GreenCurve {keys [
         {arriveTangent=3.6319, interpMode=RCIM_Cubic, leaveTangent=3.6319, time=0.2697, value=0}
         {arriveTangent=3.6319, interpMode=RCIM_Cubic, leaveTangent=3.6319, time=0.4724, value=0.736}
         {arriveTangent=-0.5306, interpMode=RCIM_Cubic, leaveTangent=-0.5306, time=0.6895, value=0.6354}
        ]}
       RedCurve {keys [
         {arriveTangent=0.6169, interpMode=RCIM_Cubic, leaveTangent=0.6169, time=0.2697, value=0.875}
         {arriveTangent=0.6169, interpMode=RCIM_Cubic, leaveTangent=0.6169, time=0.4724, value=1}
         {arriveTangent=-5.2777, interpMode=RCIM_Cubic, leaveTangent=-5.2777, time=0.6895, value=0}
        ]}
      }}))
   /SetVariables(Particles.Color=&MultiplyLinearColor(
     Linear Color=$Particles.Color
     Scale Factor=&Vector4FromCurve(
      CurveIndex=&Divide_Float(A=$Particles.NormalizedAge, B=3)
      Vector4Curve=DataInterface() {propertyValues {
        AlphaCurve {keys [<at 0 0>,<at 0.0624 1>,<at 0.3556 1>,<at 1 0>]}
        BlueCurve {keys [<at 0.0208 15>,<at 0.5255 0.9801>,<at 1 0>]}
        GreenCurve {keys [<at 0.0208 2>,<at 0.5255 2>,<at 1 0>]}
        RedCurve {keys [<at 0.0208 0>,<at 0.5255 1>,<at 1 0>]}
       }}
     )
    )) {isSetParameters=true, module=SetVariables_#1}
   &CurlNoiseForce(Noise Frequency=4, Noise Strength=255, Pan Noise=true) {module=CurlNoiseForce001}
   &SolveForcesAndVelocity(Clamp Velocity=false, Limit Acceleration=false)
   &SpriteRotationRate(Rotation Rate=&RandomRangeFloat(Minimum=1, Maximum=3))
   &ScaleSpriteSize(Uniform Curve Sprite Scale=DataInterface() {propertyValues {Curve {keys [
        {arriveTangent=8.9686, interpMode=RCIM_Cubic, leaveTangent=8.9686, time=0, value=0}
        {arriveTangent=0.9375, interpMode=RCIM_Cubic, leaveTangent=0.9375, time=0.0842, value=0.7547}
        {arriveTangent=-1.2897, interpMode=RCIM_Cubic, leaveTangent=-1.2897, tangentMode=RCTM_User, time=0.8, value=0.75}
        {arriveTangent=-0.285, interpMode=RCIM_Cubic, leaveTangent=-0.285, tangentMode=RCTM_User, time=1, value=0}
       ]}}})
  ]
  rendererClasses [&NiagaraSpriteRendererProperties]
  renderers [{rendererClass=&NiagaraSpriteRendererProperties}]
 }
 Trails {
  emitterUpdateScript [
   &EmitterState(Life Cycle Mode=::NewEnumerator1, Loop Behavior=::NewEnumerator1, Loop Duration=7, UseLoopDelay=false)
   &SpawnPerUnit(Module.Use Max Movement Threshold=false, Module.Use Movement Tolerance=false, Spawn Spacing=125, Use Spawn Probability=false)
  ]
  particleSpawnScript [
   &InitializeParticle(
    Color Mode=::NewEnumerator0
    Lifetime=3
    Mass Mode=::NewEnumerator1
    Position Offset=<0 0 0>
    Sprite Rotation Mode=::NewEnumerator2
    Sprite Size Mode=::NewEnumerator3
    Uniform Sprite Size=15
    UsePositionOffset=true
    Write Material Random=true
   ) {module=InitializeParticle001}
   &SphereLocation(
    MaskSpawnGroup=false
    Radius Position=0
    Sphere Distribution:ENiagaraSphereDistributionMode=::NewEnumerator1
    Sphere Radius=0.01
    UseNonUniformScale=false
    UseOffset=false
    UseSphereOrientation=false
    V Position=0
   )
  ]
  particleUpdateScript [
   &ParticleState()
   &GravityForce(Gravity=&Multiply_VectorByFloat(
     Vector=<0 0 3000>
     Float=&FloatFromCurve(
      FloatCurve=DataInterface() {propertyValues {Curve {keys [
          {arriveTangent=-0.1166, interpMode=RCIM_Cubic, leaveTangent=-0.1166, tangentMode=RCTM_User, time=0.8, value=1}
          {arriveTangent=-8.9384, interpMode=RCIM_Cubic, leaveTangent=-8.9384, tangentMode=RCTM_User, time=1, value=0}
         ]}}}
      CurveIndex=&ReturnNormalizedExecIndex()
     )
    ))
   &Drag(Drag=10, Use Linear Drag=true, Use Rotational Drag=false)
   &Color(Color=&ColorFromCurve(ColorCurve=DataInterface() {propertyValues {
       AlphaCurve {keys [
         {arriveTangent=0.4437, interpMode=RCIM_Cubic, leaveTangent=0.4437, tangentMode=RCTM_User, time=0.0346, value=0.25}
         {arriveTangent=-0.4336, interpMode=RCIM_Cubic, leaveTangent=-0.4336, tangentMode=RCTM_User, time=1, value=0}
        ]}
       BlueCurve {keys [
         {arriveTangent=0.6353, interpMode=RCIM_Cubic, time=0, value=0.0523}
         <at 0.1943 0.349 RCIM_Cubic>
         {interpMode=RCIM_Cubic, leaveTangent=0.024, time=0.6013, value=0.3542}
        ]}
       GreenCurve {keys [
         {arriveTangent=-0.3465, interpMode=RCIM_Cubic, time=0, value=0.4427}
         <at 0.1943 0.1768 RCIM_Cubic>
         {interpMode=RCIM_Cubic, leaveTangent=-0.5912, time=0.6013, value=0.0484}
        ]}
       RedCurve {keys [
         {arriveTangent=-2.2067, interpMode=RCIM_Cubic, time=0, value=0}
         <at 0.1943 0 RCIM_Cubic>
         {interpMode=RCIM_Cubic, leaveTangent=0.7318, time=0.6013, value=0.1589}
        ]}
      }})) {module=Color002}
   &SolveForcesAndVelocity(Clamp Velocity=false, Limit Acceleration=false)
   /SetVariables(Particles.RibbonWidth=&FloatFromCurve(
     FloatCurve=DataInterface() {propertyValues {Curve {keys [
         {arriveTangent=2.1934, interpMode=RCIM_Cubic, leaveTangent=2.1934, tangentMode=RCTM_User, time=0, value=0.3}
         {arriveTangent=0.5499, interpMode=RCIM_Cubic, leaveTangent=0.5499, tangentMode=RCTM_User, time=0.75, value=1}
        ]}}}
     Scale Curve=150
    )) {isSetParameters=true}
  ]
  rendererClasses [&NiagaraRibbonRendererProperties]
  renderers [{rendererClass=&NiagaraRibbonRendererProperties}]
 }
 single_halo {
  emitterUpdateScript [
   &EmitterState(Life Cycle Mode=::NewEnumerator1, Loop Behavior=::NewEnumerator1, Loop Duration=3, UseLoopDelay=false)
   &SpawnBurst_Instantaneous(Use Spawn Probability=false, UseLoopCountLimit=false)
  ]
  particleSpawnScript [
   &InitializeParticle(Color=<rgba 0 1 3 5>, Sprite Size Mode=::NewEnumerator3, Uniform Sprite Size=333, UsePositionOffset=false, Write Material Random=true)
   /SetVariables(Particles.Color_B:LinearColor=<rgba 0 0.1 0.1 1>) {isSetParameters=true}
  ]
  particleUpdateScript [
   /SetVariables(Particles.Scale=&VectorFromFloat(Value=&FloatFromCurve(
      FloatCurve=DataInterface() {propertyValues {Curve {keys [
          {arriveTangent=-0.0094, interpMode=RCIM_Cubic, leaveTangent=-0.0094, tangentMode=RCTM_User, time=0, value=0}
          {arriveTangent=1.25, arriveTangentWeight=0.0333, interpMode=RCIM_Cubic, leaveTangent=1.25, time=0.1, value=1}
          {arriveTangent=-1.1111, arriveTangentWeight=0.2333, interpMode=RCIM_Cubic, leaveTangent=-1.1111, time=0.8, value=1}
          <at 1 0>
         ]}}}
      Scale Curve=0.5
     ))) {isSetParameters=true}
   &ParticleState()
   /SetVariables(Particles.SpriteSize=&Multiply_Vector2D_ByFloat(
     Float=&FloatFromCurve(FloatCurve=DataInterface() {propertyValues {Curve {keys [
          {arriveTangent=-0.0094, interpMode=RCIM_Cubic, leaveTangent=-0.0094, tangentMode=RCTM_User, time=0, value=0}
          {arriveTangent=3.5889, interpMode=RCIM_Cubic, leaveTangent=3.5889, tangentMode=RCTM_User, time=0.1, value=2.125}
          {arriveTangent=-2.0839, interpMode=RCIM_Cubic, leaveTangent=-2.0839, tangentMode=RCTM_User, time=0.8, value=2}
          <at 1 0>
         ]}}})
     Vector2D=<1 1>
    )) {isSetParameters=true}
   &ScaleSpriteSize(
    Uniform Curve Scale=0.4
    Uniform Curve Sprite Scale=DataInterface() {propertyValues {Curve {keys [
        {arriveTangent=-0.0094, interpMode=RCIM_Cubic, leaveTangent=-0.0094, leaveTangentWeight=0.3333, tangentMode=RCTM_User, time=0, value=0}
        {
         arriveTangent=1.7975
         arriveTangentWeight=0.0333
         interpMode=RCIM_Cubic
         leaveTangent=1.7975
         leaveTangentWeight=0.3
         tangentMode=RCTM_User
         time=0.1
         value=0.775
        }
        {
         arriveTangent=-1.7921
         arriveTangentWeight=0.2333
         interpMode=RCIM_Cubic
         leaveTangent=-1.7921
         leaveTangentWeight=0.0667
         tangentMode=RCTM_User
         time=0.9
         value=0.7
        }
        {tangentMode=RCTM_User, time=1, value=0}
       ]}}}
   ) {module=ScaleSpriteSize001}
   &ScaleColor(
    Scale Alpha=&FloatFromCurve(
     FloatCurve=DataInterface() {propertyValues {Curve {keys [
         {arriveTangent=-0.0094, interpMode=RCIM_Cubic, leaveTangent=-0.0094, leaveTangentWeight=0.0347, tangentMode=RCTM_User, time=0, value=0}
         {
          arriveTangent=1.7975
          arriveTangentWeight=0.1245
          interpMode=RCIM_Cubic
          leaveTangent=1.7975
          leaveTangentWeight=0.6227
          tangentMode=RCTM_User
          time=0.1
          value=0.775
         }
         {
          arriveTangent=-1.7921
          arriveTangentWeight=0.1
          interpMode=RCIM_Cubic
          leaveTangent=-1.7921
          leaveTangentWeight=0.0333
          tangentMode=RCTM_User
          time=0.9
          value=0.7
         }
         {arriveTangent=-0.0383, leaveTangent=-0.0383, tangentMode=RCTM_User, time=1, value=0}
        ]}}}
     Scale Curve=0.1
    )
    ScaleA=true
    ScaleRGB=false
   ) {module=ScaleColor001}
   &CameraOffset(Camera Offset Amount=15)
  ]
  rendererClasses [&NiagaraMeshRendererProperties]
  renderers [{rendererClass=&NiagaraMeshRendererProperties, rendererIndex=0}]
 }
}
summary {
 systemName=NS_GiftBox_SparkleTrails
 emitters [
  {emitterName=Trails, rendererClasses [&NiagaraRibbonRendererProperties]}
  {emitterName=single_halo, rendererClasses [&NiagaraMeshRendererProperties]}
  {emitterName=Confetti, rendererClasses [&NiagaraSpriteRendererProperties]}
  {emitterName=Confetti001, rendererClasses [&NiagaraSpriteRendererProperties]}
  {emitterName=Sparkes, rendererClasses [&NiagaraSpriteRendererProperties]}
 ]
 userVariables []
}
